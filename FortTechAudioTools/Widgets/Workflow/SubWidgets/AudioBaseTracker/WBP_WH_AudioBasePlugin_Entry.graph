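# List entry widget for the AudioBase tracker in the FortTechAudioTools workflow UI, showing one plugin name as text in a dark, fixed-height row.
# When its list item object is set, Initialize casts it to this same widget class and copies its NewName into the AudiobaseText label.
# List items are therefore instances of the entry class itself.
# An AssetClassEntryAdded dispatcher is declared but never called here.
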WidgetBlueprint /FortTechAudioTools/Widgets/Workflow/SubWidgets/AudioBaseTracker/WBP_WH_AudioBasePlugin_Entry
graphs [
 AssetClassEntryAdded {nodes [AssetClassEntryAdded!() -> (then, AssetClass:Object Class Reference) {is=title K2Node_FunctionEntry}]}
 EventGraph {nodes [
   n1 = Event OnListItemObjectSet()
   > Initialize!(execute, self, Object:Object Reference=@n1.ListItemObject) -> (then)
  ]}
 Initialize {nodes [
   n1 = Initialize!() -> (then, Object:Object Reference) {is=title K2Node_FunctionEntry}
   > n2 = Cast To WBP_WH_AudioBasePlugin_Entry!(execute, Object=@n1) -> (then, CastFailed, AsWBP WH Audio Base Plugin Entry:WBP WH Audio Base Plugin Entry Object Reference) {is=title K2Node_DynamicCast}
   > Set Text(
    Text=Get NewName!(self:WBP WH Audio Base Plugin Entry Object Reference=@n2) -> (NewName:Text)
    self=Get AudiobaseText() -> (AudiobaseText:Common Text Object Reference)
   )
  ]}
]
variables [
 AssetClassEntryAdded {}
 NewName {type=string, default=""}
]
widgets {
 info {inheritedWidgetCount=0, parentClass=&UserWidget, rootWidgetClass=&CanvasPanel, widgetCount=4}
 widgets [
  CanvasPanel_0 {widgetClassPath=&CanvasPanel, widget=CanvasPanel_0, properties {bOverride_Cursor=true}}
  SizeBox_0 {
   widgetClassPath=&SizeBox
   parent=CanvasPanel_0
   widget=SizeBox_0
   properties {bOverride_HeightOverride=true, heightOverride=35, widthOverride=500}
   slot {
    class=CanvasPanelSlot
    properties {bAutoSize=true, layoutData {alignment=<0 0>, anchors {maximum=<1 1>, minimum=<0 0>}, offsets=<ltrb 0 0 0 0>}}
   }
  }
  Border_63 {
   widgetClassPath=&Border
   parent=SizeBox_0
   widget=Border_63
   properties {
    background {
     bIsDynamicallyLoaded=false
     drawAs=Box
     imageType=NoImage
     mirroring=NoMirror
     resourceName=None
     resourceObject=null
     tiling=NoTile
     imageSize=<32 32>
     margin=<ltrb 1 1 1 1>
     outlineSettings {
      bUseBrushTransparency=false
      roundingType=HalfHeightRadius
      width=0
      color {colorUseRule=UseColor_Specified, specifiedColor=<rgba 0 0 0 0>}
      cornerRadii=<0 0 0 0>
     }
     tintColor {colorUseRule=UseColor_Specified, specifiedColor=<rgba 0.0103 0.0103 0.0103 1>}
     uVRegion {bIsValid=false, max=<0 0>, min=<0 0>}
    }
    padding=<ltrb 10 0 10 0>
   }
   slot {class=SizeBoxSlot, properties {padding=<ltrb 0 0 10 -10>}}
  }
  AudiobaseText {
   widgetClassPath=&CommonTextBlock
   bIsVariable=true
   parent=Border_63
   widget=AudiobaseText
   properties {
    style=&TextStyle-Base_C
    text=Plugin
    font {
     bForceMonospaced=false
     bMaterialIsStencil=false
     fontMaterial=null
     fontObject=&NotoSans
     letterSpacing=0
     monospacedWidth=1
     size=24
     skewAmount=0
     typefaceFontName=NotoSans-Regular
     outlineSettings {
      bApplyOutlineToDropShadows=false
      bMiteredCorners=false
      bSeparateFillAlpha=false
      outlineMaterial=null
      outlineSize=0
      outlineColor=<rgba 0 0 0 1>
     }
    }
    shadowOffset=<0 0>
   }
   slot {class=BorderSlot, properties {padding=<ltrb 10 0 10 0>}}
  }
 ]
}
